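# Translucent UI material for the Magneto powers reticle, drawing brackets and a shield with adjustable offsets.
# Mirrored channel-packed distance-field textures produce bracket fill plus shield fill and outline.
# Shield VFX multiplies rotating, panning noise by animated radial sine bands and a mask, while start, end and animate parameters drive a progress gradient.
# Layered lerps tint white, black and lavender-blue, with ShieldOpacity and HideBrackets controlling visibility.
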Material /TwinStretchCorn/UI/Art/M_UI_MagnettoPowersReticle
graph {
 properties {blendMode=BLEND_Translucent, materialDomain=MD_UI, pixelDepthOffsetMode=PDOM_Legacy, usedShadingModels=Default Lit}
 nodes [
  n1 = Multiply(A=NamedRerouteUsage()) {constB=0.3}
  n2 = Min(A=NamedRerouteUsage(), B=NamedRerouteUsage())
  n3 = Multiply(A=NamedRerouteUsage()) {constB=1.5}
  n4 = MaterialFunctionCall(GradientStart=Constant() {r=-0.02}, GradientEnd=Constant() {r=1}) {materialFunction=&MF_UI_RadialGradient}
  Add(
   A=Multiply(A=LinearInterpolate(A=@n1, B=@n2, Alpha=@n2), B=ScalarParameter() {defaultValue=0.7, parameterName=ShieldOpacity})
   B=LinearInterpolate(B=Multiply(A=@n3, B=@n4.Smooth), Alpha=ScalarParameter() {parameterName=HideBrackets})
  )
  Constant2Vector() {r=-0.376}
  n5 = Constant3Vector() {constant=<rgba 1 1 1 0>}
  n6 = Constant3Vector()
  n7 = NamedRerouteUsage()
  n8 = Constant3Vector() {constant=<rgba 0.3231 0.2542 1 1>}
  n9 = Constant3Vector() {constant=<rgba 0.603 0.5625 1 1>}
  n10 = NamedRerouteUsage()
  n11 = NamedRerouteUsage()
  LinearInterpolate(
   A=LinearInterpolate(A=LinearInterpolate(A=@n5, B=@n6, Alpha=@n7), B=LinearInterpolate(A=@n8, B=@n9, Alpha=@n10), Alpha=NamedRerouteUsage())
   B=Constant3Vector() {constant=<rgba 0.3855 0.3229 1 1>}
   Alpha=Saturate(None=Multiply(A=@n11) {constB=1.5})
  )
  n12 = ScalarParameter() {defaultValue=0.272, parameterName=OffsetXShield}
  n13 = Constant()
  n14 = TextureSample(UVs=MaterialFunctionCall!(UV, Offset=AppendVector(A=@n12, B=@n13), Rotation, Scale=Constant2Vector() {g=0.5, r=1}, MirrorY) {materialFunction=/FortUILibrary/MaterialFunctions/Basics/MF_UI_Symmetry_2x1_DEPRECATED}.Result) {samplerSource=SSM_Clamp_WorldGroupSettings, texture=/TwinStretchCorn/UI/Art/Textures/T_UI_Reticle_Magneto}
  n15 = MaterialFunctionCall!(RGBA Channels=@n14.G, Glow Min=Constant() {r=0.252}, Glow Max=Constant() {r=0.316}, Outline Thickness) {materialFunction=/FortUILibrary/MaterialFunctions/Basics/MF_UI_ChannelPackedDistanceField}
  n16 = MaterialFunctionCall(RGBA Channels=@n14.G, Glow Min=Constant() {r=0.6}, Glow Max=Constant() {r=0.8}) {materialFunction=&MF_UI_ChannelPackedDistanceField}
  NamedRerouteDeclaration(None=Saturate(None=Subtract(A=@n15.Fill, B=@n16.Outline))) {name=ShieldOutline, nodeColor=<rgba 0 0 0 1>}
  n17 = MaterialFunctionCall(UVs=Panner(Time=Time()) {speedX=0.5, speedY=0.5}, "Rotation Angle (0-1)"=Time()) {materialFunction=&CustomRotator}
  n18 = ConstantBiasScale(None=TextureCoordinate()) {bias=-0.5, scale=2}
  n19 = Constant()
  n20 = ComponentMask(None=Panner(Coordinate=Distance(A=@n18, B=@n19)) {speedX=-0.24}) {r=true}
  n21 = OneMinus(None=SmoothStep(Value=Sine(None=@n20) {period=0.25}) {constMax=0.7, constMin=1})
  n22 = MaterialFunctionCall(Offset=Constant2Vector() {r=-0.35}, Scale=Constant2Vector() {g=1, r=2}) {materialFunction=&MF_UI_Symmetry_2x1_DEPRECATED}
  n23 = Constant() {r=0.5}
  n24 = MaterialFunctionCall(RGBA Channels=TextureSample(UVs=MaterialFunctionCall(UVs=@n22.Result, "Rotation Angle (0-1)"=@n23) {materialFunction=&CustomRotator}.Rotated Values) {texture=/TwinStretchCorn/UI/Art/Textures/T_UI_MagnetoMask}.R) {materialFunction=&MF_UI_ChannelPackedDistanceField}
  NamedRerouteDeclaration(None=Multiply(
    A=TextureSample(UVs=@n17.Rotated Values) {samplerType=SAMPLERTYPE_Grayscale, texture=/Game/TimeOfDay/TODM_Sequence/Materials/Sky/Textures/LowResBlurredNoise_Radish_LF_Root}.R
    B=Multiply(A=@n21, B=@n24.Fill)
   )) {name=ShieldVFX, nodeColor=<rgba 0 0 0 1>}
  n25 = MaterialFunctionCall(Offset=AppendVector(A=ScalarParameter() {defaultValue=0.06, parameterName=OffsetX}, B=Constant()), Scale=Constant2Vector() {g=0.5, r=1}) {materialFunction=&MF_UI_Symmetry_2x1_DEPRECATED}
  NamedRerouteDeclaration(None=MaterialFunctionCall(
    RGBA Channels=TextureSample(UVs=@n25.Result) {samplerSource=SSM_Clamp_WorldGroupSettings, texture=&T_UI_Reticle_Magneto}.R
    Glow Min=Constant() {r=0.252}
    Glow Max=Constant() {r=0.316}
   ) {materialFunction=&MF_UI_ChannelPackedDistanceField}.Fill) {name=BacketsFill, nodeColor=<rgba 0 0 0 1>}
  NamedRerouteDeclaration(None=@n15.Fill) {name=ShieldFill, nodeColor=<rgba 0 0 0 1>}
  n26 = ScalarParameter() {group=Progress_Layout, parameterName=ProgressEdgeSharpness, sliderMax=1, sortPriority=2}
  n27 = MaterialFunctionCall(UVs=TextureCoordinate(), "Rotation Angle (0-1)"=Constant() {r=0.5}) {materialFunction=&CustomRotator}
  n28 = ScalarParameter() {
   defaultValue=0.33
   desc="When progress value changes, it defines where change starts.\r\n0 - 1"
   group=Progress
   parameterName=ProgressStart
   sliderMax=1
   sortPriority=1
  }
  n29 = ScalarParameter() {
   defaultValue=0.66
   desc="When progress value changes, it defines where change ends.\r\n0 - 1"
   group=Progress
   parameterName=ProgressEnd
   sliderMax=1
   sortPriority=2
  }
  n30 = ScalarParameter() {
   defaultValue=1
   desc="Property for animating progress value change (from ProgressStart to ProgressEnd)\r\n0 - 1"
   group=Progress
   parameterName=AnimateProgress
   sliderMax=1
   sortPriority=3
  }
  n31 = LinearInterpolate(A=@n28, B=@n29, Alpha=@n30)
  n32 = ComponentMask(None=Add(A=Multiply(A=@n27.Rotated Values) {constB=-1}, B=NamedRerouteDeclaration(None=@n31) {name=Progress, nodeColor=<rgba 0 0 0 1>}) {constB=0.5}) {g=true}
  NamedRerouteDeclaration(None=Saturate(None=SmoothStep(Max=@n26, Value=@n32) {constMax=0.01})) {name=ProgressGraident, nodeColor=<rgba 0 0 0 1>}
 ]
}
