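# Translucent UI material that draws the Magnetto attachment point marker as a pair of brackets and a small center dot.
# Brackets come from a horizontally squeezed SDF circle outline masked by two rotated radial segments; the dot is a small circle fill.
# IsTargettingVehicle fades in a half-strength mask that is added to another for opacity.
# IsAllReady shifts masked color from white to periwinkle blue.
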Material /TwinStretchCorn/UI/Art/M_UI_MagnettoAttachmentPoint
graph {
 properties {
  bAutomaticallySetUsageInEditor=false
  blendMode=BLEND_Translucent
  materialDomain=MD_UI
  pixelDepthOffsetMode=PDOM_Legacy
  usedShadingModels=Default Lit
 }
 nodes [
  n1 = NamedRerouteUsage()
  n2 = Constant() {r=0.5}
  Add(A=LinearInterpolate(B=Multiply(A=@n1, B=@n2), Alpha=ScalarParameter() {parameterName=IsTargettingVehicle}), B=NamedRerouteUsage())
  n3 = Constant3Vector() {constant=<rgba 1 1 1 0>}
  LinearInterpolate(
   A=@n3
   B=LinearInterpolate(A=@n3, B=Constant3Vector() {constant=<rgba 0.3231 0.2542 1 1>}, Alpha=ScalarParameter() {parameterName=IsAllReady})
   Alpha=NamedRerouteUsage()
  )
  n4 = TextureCoordinate()
  n5 = Add(A=Multiply(A=ComponentMask(None=@n4) {r=true}, B=Constant() {r=1.2}), B=Constant() {r=-0.1})
  n6 = ComponentMask(None=@n4) {g=true}
  n7 = SmoothStep(Value=MaterialFunctionCall(UVs=AppendVector(A=@n5, B=@n6), OutlineThickness=Constant() {r=0.02}) {materialFunction=&MF_UI_Circle}.SDFOutline) {constMax=0.01}
  n8 = MaterialFunctionCall(UVs=TextureCoordinate(), "Rotation Angle (0-1)"=Constant() {r=0.25}) {materialFunction=&CustomRotator}
  n9 = Constant() {r=2}
  n10 = Constant() {r=0.324}
  NamedRerouteDeclaration(None=Multiply(
    A=OneMinus(None=@n7)
    B=MaterialFunctionCall!(UV=@n8.Rotated Values, Segment Count=@n9, Segment Spacing=@n10, Fixed Spacing) {materialFunction=/FortUILibrary/MaterialFunctions/Basics/MF_UI_RadialSegments}.Segmented Triangles Fill
   )) {name=Brackets, nodeColor=<rgba 0 0 0 1>}
  NamedRerouteDeclaration(None=MaterialFunctionCall(Size=Constant2Vector() {g=0.03, r=0.03}) {materialFunction=&MF_UI_Circle}.Fill) {name=MiddleDot, nodeColor=<rgba 0 0 0 1>}
 ]
}
