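# Reticle widget for the Magnetto powers attachment point that shows whether a vehicle is targeted.
# An overlay holds a single image using the M_UI_MagnettoAttachmentPoint material.
# The TargettingVehicle function plays a short animation forward when a vehicle is found and in reverse otherwise.
# That animation drives the material's IsTargettingVehicle parameter from 0 to 1, so the material renders the targeting transition.
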WidgetBlueprint /TwinStretchCorn/UI/WBP_MagnettoPowers_AttachmentPoint
animations [Anim_TargettedVehicle {
  binding_count=1
  length_seconds=0.25
  object=&self:Anim_TargettedVehicle
  sequence {bindings [Image_ReticlePointOfAttachment_1 {
     id=#1
     possessed_class=Image
     tracks [{
       type=MovieSceneWidgetMaterialTrack
       sections [{
         type=MovieSceneParameterSection
         end_frame=5
         start_frame=0
         channels [IsTargettingVehicle {type=MovieSceneScriptingFloatChannel, keys [[0,0],[5,1]]}]
        }]
      }]
    }]}
 }]
graphs [
 EventGraph {links [], nodes []}
 TargettingVehicle {nodes [
   n1 = TargettingVehicle!() -> (then, FoundVehicle:bool=~) {is=title K2Node_FunctionEntry}
   > n2 = Branch(Condition=@n1)
   n3 = Get Anim_TargettedVehicle() -> (Anim_TargettedVehicle:Widget Animation Object Reference)
   > PlayAnimationForward(InAnimation=@n3)
   PlayAnimationReverse(execute=@n2.else, InAnimation=@n3)
  ]}
]
variables []
widgets {
 info {inheritedWidgetCount=0, parentClass=&UserWidget, rootWidgetClass=&Overlay, widgetCount=2}
 widgets [
  Overlay_LockedOn {widgetClassPath=&Overlay, bIsVariable=true, widget=Overlay_LockedOn}
  Image_ReticlePointOfAttachment_1 {
   widgetClassPath=&Image
   bIsVariable=true
   parent=Overlay_LockedOn
   widget=Image_ReticlePointOfAttachment_1
   properties {
    brush {
     bIsDynamicallyLoaded=false
     drawAs=Image
     imageType=FullColor
     mirroring=NoMirror
     resourceName=None
     resourceObject=/TwinStretchCorn/UI/Art/M_UI_MagnettoAttachmentPoint
     tiling=NoTile
     imageSize=<45 45>
     margin=<ltrb 0 0 0 0>
     outlineSettings {
      bUseBrushTransparency=false
      roundingType=HalfHeightRadius
      width=0
      color {colorUseRule=UseColor_Specified, specifiedColor=<rgba 0 0 0 0>}
      cornerRadii=<0 0 0 0>
     }
     tintColor {colorUseRule=UseColor_Specified, specifiedColor=<rgba 1 1 1 1>}
     uVRegion {bIsValid=false, max=<0 0>, min=<0 0>}
    }
    renderTransform {angle=0, scale=<1.5 1.5>, shear=<0 0>, translation=<0 0>}
   }
   slot {class=OverlaySlot, properties {horizontalAlignment=HAlign_Fill, verticalAlignment=VAlign_Fill}}
  }
 ]
}
