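# Reticle widget for the Magnetto Powers weapon (B_Weap_TwinStretchCorn), extending the recharge-stack reticle base with pip, shield and vehicle-lock feedback.
# Weapon events switch the material-animated reticle for blocking and pass vehicle targeting to the attachment-point widget.
# Shield health drives a material parameter, with the one-time binding retried on blocking because the shield may not exist yet.
# A timer polls cooldown for the pips while the magazine is below full.
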WidgetBlueprint /TwinStretchCorn/UI/WBP_MagnettoPowersReticle
animations [
 Intro {
  binding_count=1
  length_seconds=0.5
  object=&self:Intro
  sequence {bindings [Root {
     id=#1
     possessed_class=Overlay
     tracks [
      {
       type=MovieScene2DTransformTrack
       sections [{
         type=MovieScene2DTransformSection
         end_frame=10
         start_frame=0
         channels [
          Scale.X {type=MovieSceneScriptingFloatChannel, keys [[0,1.25],[10,1]]}
          Scale.Y {type=MovieSceneScriptingFloatChannel, keys [[0,1.25],[10,1]]}
         ]
        }]
      }
      {
       type=MovieSceneFloatTrack
       display_name=Render Opacity
       sections [{type=MovieSceneFloatSection, end_frame=10, start_frame=0, channels [None {type=MovieSceneScriptingFloatChannel, keys [[0,0],[10,1]]}]}]
      }
     ]
    }]}
 }
 Anim_SwitchState {
  binding_count=1
  length_seconds=0.5
  object=&self:Anim_SwitchState
  sequence {bindings [Reticle {
     id=#2
     possessed_class=Image
     tracks [{
       type=MovieSceneWidgetMaterialTrack
       sections [{
         type=MovieSceneParameterSection
         end_frame=10
         start_frame=0
         channels [
          HideBrackets {type=MovieSceneScriptingFloatChannel, keys [[3,0],[10,1]]}
          OffsetXShield {type=MovieSceneScriptingFloatChannel, keys [[0,0.272],[10,-0.188]]}
          ShieldOpacity {type=MovieSceneScriptingFloatChannel, keys [[0,0.7],[10,2]]}
         ]
        }]
      }]
    }]}
 }
]
graphs [
 BindToWeaponEvents {nodes [
   n1 = BindToWeaponEvents!() -> (then, DoBind:bool, EquippedWeapon:Fort Weapon Object Reference) {is=title K2Node_FunctionEntry}
   > Parent: BindToWeaponEvents!(execute, DoBind:bool=@n1, EquippedWeapon:Fort Weapon Object Reference=@n1.EquippedWeapon) -> (then) {is=title K2Node_CallParentFunction}
   > n2 = Cast To B_Weap_TwinStretchCorn!(execute, Object=@n1.EquippedWeapon) -> (then, CastFailed, AsB Weap Twin Stretch Corn:B Weap Twin Stretch Corn Object Reference) {is=title K2Node_DynamicCast}
   > n3 = Set MagnettoPowersWeaponRef(MagnettoPowersWeaponRef:B Weap Twin Stretch Corn Object Reference=@n2) -> (Output_Get:B Weap Twin Stretch Corn Object Reference)
   > Bind Event to On Blocking State Changed(self:B Weap Twin Stretch Corn Object Reference=@n3, Delegate=Create Event()) {is=title K2Node_AddDelegate}
   > Try Bind to Shield Events()
   Bind Event to Get Locked on Vehicle(
    self:B Weap Twin Stretch Corn Object Reference=Get MagnettoPowersWeaponRef() -> (MagnettoPowersWeaponRef:B Weap Twin Stretch Corn Object Reference)
    Delegate=Create Event()
   ) {is=title K2Node_AddDelegate}
  ]}
 EventGraph {nodes [
   Event Construct()
   > Parent: Construct()
   > PlayAnimation(InAnimation=Get Intro())
   > UpdateCooldownQueryTimer(self:Fort Weapon Reticle Extension Widget Base Object Reference)
   Event Destruct()
   > Parent: Destruct()
   > HandleCooldownTimer!(execute, self:Fort Weapon Reticle Extension Widget Base Object Reference, bActivateTimer:bool=false, TimerInterval:float=0) -> (then)
   n1 = Event OnLocalAmmoChanged!() -> (OutputDelegate, then, LocalCount:int=0, LocalRemaining:int=0) {is=title K2Node_Event}
   > Parent: OnLocalAmmoChanged!(execute, LocalCount:int=@n1.LocalCount, LocalRemaining:int=@n1.LocalRemaining) -> (then) {is=title K2Node_CallParentFunction}
   > UpdateCooldownQueryTimer()
   Event UpdateRemainingCooldown() {is=title K2Node_Event}
   > n2 = GetRemainingCooldownTimer!(execute, self) -> (then, RemainingTime:double, TotalDuration:double)
   > UpdateRemainingRechargeTime!(
    execute
    self:Fort Weapon Reticle Recharge Stack Base Pip Collection Base Object Reference=Get WBP_MagnettoPowersReticle_Collection() -> (WBP_MagnettoPowersReticle_Collection:WBP Magnetto Powers Reticle Collection Object Reference)
    TimeRemaining:float=@n2
    TotalDuration:float=@n2.TotalDuration
    PipPercentMatParamName:Name=Percent
    ShouldSetPipColor:bool=false
    NotReadyColor:Linear Color Structure
    ReadyColor:Linear Color Structure
    PipColorMatParamName:Name=None
   ) -> (then)
   Event UpdateCooldownQueryTimer() {is=title K2Node_Event}
   > n3 = Get!(execute:Exec) -> (then:Exec, else:Exec, MagnettoPowersWeaponRef:B Weap Twin Stretch Corn Object Reference)
   > n4 = Branch(Condition="integer >= integer"(
     A=GetLocalMagazineAmmoCount!(self:Fort Weapon Object Reference=@n3) -> (ReturnValue:int=0)
     B=GetBulletsPerClip!(self:Fort Weapon Object Reference=@n3) -> (ReturnValue:int=0)
    ))
   > HandleCooldownTimer(bActivateTimer=false, TimerInterval=0)
   HandleCooldownTimer(execute=@n4.else, bActivateTimer=true, TimerInterval=0.1)
  ]}
 GetPipCollection {nodes [
   GetPipCollection() {is=title K2Node_FunctionEntry}
   > Return Node!(execute, PipCollection:WBP Reticle Recharge Stack Base Pip Collection Object Reference=Get WBP_MagnettoPowersReticle_Collection())
  ]}
 GetRemainingCooldownTimer {nodes [
   GetRemainingCooldownTimer() {is=title K2Node_FunctionEntry}
   > n1 = Cast To FortPlayerController(Object=GetOwningPlayer())
   > n2 = Get()
   n3 = GetCoolDownTimeRemaining!(
    self:Fort Item Object Reference=GetInventoryItem!(self:Fort Weapon Object Reference=@n2) -> (ReturnValue:Fort Item Object Reference)
    FortPC:Fort Player Controller Object Reference=@n1
   ) -> (CooldownDuration:float, ReturnValue:float)
   > Return Node!(execute, RemainingTime:double=@n3.ReturnValue, TotalDuration:double=@n3)
   Return Node(execute=[@n2.else, @n1.CastFailed], RemainingTime=0, TotalDuration=0)
  ]}
 SetBlockingState {nodes [
   n1 = SetBlockingState!() -> (then, IsBlocking:bool=~) {is=title K2Node_FunctionEntry}
   > n2 = Branch(Condition=@n1)
   n3 = Get Anim_SwitchState() -> (Anim_SwitchState:Widget Animation Object Reference)
   > PlayAnimationForward(InAnimation=@n3)
   > Try Bind to Shield Events()
   PlayAnimationReverse(execute=@n2.else, InAnimation=@n3)
  ]}
 SetChargedItems {nodes [
   n1 = GetAllEntries!(self:Dynamic Entry Box Base Object Reference=GetDynamicEntryBox!(self:Fort Weapon Reticle Recharge Stack Base Pip Collection Base Object Reference=GetPipCollection!(self:WBP Reticle Recharge Stack Base Object Reference) -> (PipCollection:WBP Reticle Recharge Stack Base Pip Collection Object Reference) {is=title K2Node_CallFunction}) -> (ReturnValue:Dynamic Entry Box Object Reference)) -> (ReturnValue:Array of User Widget Object References)
   Set EntriesNumber(EntriesNumber:int=Length(TargetArray:Array of User Widget Object References=@n1)) -> (Output_Get:int=0)
   n2 = Get EntriesNumber() -> (EntriesNumber:int=0)
   n3 = Get Num() -> (Num:int=0)
   > n4 = Cast To WBP_MagnettoPowersReticle_Pip!(
    execute
    Object="Get (a copy)"(Array:Array of User Widget Object References=@n1, Dimension 1=int - int(A=@n2, B=@n3)) -> (Output:User Widget Object Reference)
   ) -> (then, CastFailed, AsWBP Magnetto Powers Reticle Pip:WBP Magnetto Powers Reticle Pip Object Reference) {is=title K2Node_DynamicCast}
   > Play Charged Animation(self:WBP Magnetto Powers Reticle Pip Object Reference=@n4)
   SetChargedItems!() -> (then, Num:int) {is=title K2Node_FunctionEntry}
  ]}
 SetShieldHealthValue {nodes [
   SetShieldHealthValue() {is=title K2Node_FunctionEntry}
   > n1 = GetDynamicMaterial(self=Get Reticle() -> (Reticle:Image Object Reference))
   n2 = Get MagnettoPowersWeaponRef()
   > SetScalarParameterValue(
    self=@n1
    ParameterName=AnimateProgress
    Value=GetHealthPercent!(self:Building Actor Object Reference=Get SpawnedShield!(self:B Weap Twin Stretch Corn Object Reference=@n2) -> (SpawnedShield:BGA Twin Stretch Corn Block Volume Object Reference)) -> (ReturnValue:float)
   )
  ]}
 Try Bind to Shield Events {nodes [
   Try Bind to Shield Events() {is=title K2Node_FunctionEntry}
   n1 = Get SpawnedShield(self=Get MagnettoPowersWeaponRef())
   > Is Valid(InputObject=@n1)
   > Do Once()
   > Bind Event to On Building Health Changed(self:Building Actor Object Reference=@n1, Delegate=Create Event()) {is=title K2Node_AddDelegate}
  ]}
 VehicleLocked {nodes [
   n1 = VehicleLocked!() -> (then, LockedOnVehicle:Actor Object Reference, FoundTarget:bool=~) {is=title K2Node_FunctionEntry}
   > TargettingVehicle!(
    execute
    self:WBP Magnetto Powers Attachment Point Object Reference=Get WBP_MagnettoPowers_AttachmentPoint() -> (WBP_MagnettoPowers_AttachmentPoint:WBP Magnetto Powers Attachment Point Object Reference)
    FoundVehicle:bool=@n1.FoundTarget
   ) -> (then)
  ]}
]
variables [MagnettoPowersWeaponRef {type=B_Weap_TwinStretchCorn_C, default=null}]
widgets {
 info {
  inheritedWidgetCount=1
  parentClass=/Game/Athena/HUD/reticles/ReticleExtensionWidgets/ChargeStack/WBP_Reticle_RechargeStackBase.WBP_Reticle_RechargeStackBase_C
  rootWidgetClass=&Overlay
  widgetCount=5
 }
 widgets [
  Root {widgetClassPath=&Overlay, bIsVariable=true, widget=Root}
  WBP_MagnettoPowersReticle_Collection {
   widgetClassPath=&/WBP_MagnettoPowersReticle_Collection
   bIsVariable=true
   parent=Root
   widget=WBP_MagnettoPowersReticle_Collection
   slot {class=OverlaySlot, properties {horizontalAlignment=HAlign_Center, verticalAlignment=VAlign_Bottom, padding=<ltrb 0 140 0 0>}}
  }
  Reticle {
   widgetClassPath=&Image
   bIsVariable=true
   parent=Root
   widget=Reticle
   properties {brush {
     bIsDynamicallyLoaded=false
     drawAs=Image
     imageType=FullColor
     mirroring=NoMirror
     resourceName=None
     resourceObject=/TwinStretchCorn/UI/Art/M_UI_MagnettoPowersReticle
     tiling=NoTile
     imageSize=<180 180>
     margin=<ltrb 0 0 0 0>
     outlineSettings {
      bUseBrushTransparency=false
      roundingType=HalfHeightRadius
      width=0
      color {colorUseRule=UseColor_Specified, specifiedColor=<rgba 0 0 0 0>}
      cornerRadii=<0 0 0 0>
     }
     tintColor {colorUseRule=UseColor_Specified, specifiedColor=<rgba 1 1 1 1>}
     uVRegion {bIsValid=false, max=<0 0>, min=<0 0>}
    }}
   slot {class=OverlaySlot, properties {horizontalAlignment=HAlign_Center, verticalAlignment=VAlign_Center}}
  }
  WBP_MagnettoPowers_AttachmentPoint {
   widgetClassPath=&/WBP_MagnettoPowers_AttachmentPoint
   bIsVariable=true
   parent=Root
   widget=WBP_MagnettoPowers_AttachmentPoint
   slot {class=OverlaySlot, properties {horizontalAlignment=HAlign_Center, verticalAlignment=VAlign_Center}}
  }
  TutorialHighlight {widgetClassPath=&TutorialHighlight}
 ]
}
